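# Physics asset for the SatinCheddar pickaxe, a cosmetic melee weapon mesh.
# It has a single body on the Melee_TwoHanded_Handle bone, made of one offset, rotated capsule.
# It defines no constraints, so the physics representation is one simple shape with no articulated parts.
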PhysicsAsset /BRCosmetics/Weapons/FORT_Melee/Pickaxe_SatinCheddar/Meshes/SatinCheddar_Axe_PhysicsAsset
bodies {Melee_TwoHanded_Handle [{shapeType=Capsule, length=22.8778, radius=21.8179, center=<0.6096 -1.3648 -1.1331>, rotation=<rot 0 161.2724 -61.7059>}]}
constraints []
